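# Control Rig for the MajorSpeech backpack cosmetic that procedurally waves its eight ribbon bone chains.
# The forward solve runs one ChainHarmonicsPerItem per left/right ribbon (main, in, dw, mid), using only waves, with pendulum and reach disabled.
# Skydive_WaveSpeed, default 0, times a signed per-chain factor drives each chain's Y or Z wave speed, varying the motion between ribbons.
# The Skydive bool is declared but unused in the graph.
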ControlRigBlueprint /BRCosmetics/Accessories/FORT_Backpacks/Backpack_F_MED_MajorSpeech/Meshes/F_MED_MajorSpeech_Pack_CtrlRig
graphs [
 RigVMFunctionLibrary {links [], nodes []}
 RigVMModel {nodes [
   RigUnit_BeginExecution = &RigUnit_BeginExecution()
   Multiply = &RigVMFunction_MathDoubleMul(A=1.3, B=RigVMVariableNode() {variable=Skydive_WaveSpeed}.Value)
   /Script/ControlRig.RigUnit_ChainHarmonicsPerItem(
    ChainRoot=(Type=Bone,Name="dyn_Ribbon_l_01")
    DrawWorldOffset=(Rotation=(X=0,Y=0,Z=0,W=1),Translation=(X=0,Y=0,Z=0),Scale3D=(X=1,Y=1,Z=1))
    Pendulum=(bEnabled=false,PendulumStiffness=50,PendulumGravity=(X=-900,Y=-500,Z=0),PendulumBlend=1,PendulumDrag=20,PendulumMinimum=0,PendulumMaximum=0.8,PendulumEase=SineEaseInOut,UnwindAxis=(X=0,Y=0,Z=0),UnwindMinimum=0,UnwindMaximum=0)
    Reach=(bEnabled=false,ReachTarget=(X=0,Y=0,Z=0),ReachAxis=(X=1,Y=1,Z=1),ReachMinimum=0,ReachMaximum=0,ReachEase=SineEaseInOut)
    Speed=(X=0,Y=1.3,Z=0)
    Wave=(bEnabled=True,WaveFrequency=(X=0,Y=6,Z=0),WaveAmplitude=(X=0,Y=-2.4,Z=0),WaveOffset=(X=0,Y=1,Z=0),WaveNoise=(X=0,Y=0,Z=0),WaveMinimum=0.1,WaveMaximum=0.6,WaveEase=SineEaseInOut)
    WaveCurve="(EditorCurveData=(Keys=((InterpMode=RCIM_Linear,TangentMode=RCTM_Auto,TangentWeightMode=RCTWM_WeightedNone,Time=0,Value=0.2,ArriveTangent=0,ArriveTangentWeight=0,LeaveTangent=0,LeaveTangentWeight=0),(InterpMode=RCIM_Linear,TangentMode=RCTM_Auto,TangentWeig ...(truncated)"
    bDrawDebug=True
    Speed.Y=@Multiply.Result
    ExecutePin=@RigUnit_BeginExecution.ExecutePin
   ) {is=struct RigVMUnitNode}
   > &RigUnit_ChainHarmonicsPerItem(
    ChainRoot=(Type=Bone,Name="dyn_Ribbon_r_01")
    DrawWorldOffset=(Rotation=(X=0,Y=0,Z=0,W=1),Translation=(X=0,Y=0,Z=0),Scale3D=(X=1,Y=1,Z=1))
    Pendulum=(bEnabled=false,PendulumStiffness=50,PendulumGravity=(X=-900,Y=-500,Z=0),PendulumBlend=1,PendulumDrag=20,PendulumMinimum=0,PendulumMaximum=0.8,PendulumEase=SineEaseInOut,UnwindAxis=(X=0,Y=0,Z=0),UnwindMinimum=0,UnwindMaximum=0)
    Reach=(bEnabled=false,ReachTarget=(X=0,Y=0,Z=0),ReachAxis=(X=1,Y=1,Z=1),ReachMinimum=0,ReachMaximum=0,ReachEase=SineEaseInOut)
    Speed=(X=0,Y=-1,Z=0)
    Wave=(bEnabled=True,WaveFrequency=(X=0,Y=7,Z=0),WaveAmplitude=(X=0,Y=-3,Z=0),WaveOffset=(X=0,Y=1,Z=0),WaveNoise=(X=0,Y=0,Z=0),WaveMinimum=0.1,WaveMaximum=0.6,WaveEase=SineEaseInOut)
    WaveCurve="(EditorCurveData=(Keys=((InterpMode=RCIM_Linear,TangentMode=RCTM_Auto,TangentWeightMode=RCTWM_WeightedNone,Time=0,Value=0.2,ArriveTangent=0,ArriveTangentWeight=0,LeaveTangent=0,LeaveTangentWeight=0),(InterpMode=RCIM_Linear,TangentMode=RCTM_Auto,TangentWeig ...(truncated)"
    bDrawDebug=True
    Speed.Y=&RigVMFunction_MathDoubleMul(A=-1, B=RigVMVariableNode() {variable=Skydive_WaveSpeed}.Value).Result
   )
   > &RigUnit_ChainHarmonicsPerItem(
    ChainRoot=(Type=Bone,Name="dyn_Ribbon_l_in_01")
    DrawWorldOffset=(Rotation=(X=0,Y=0,Z=0,W=1),Translation=(X=0,Y=0,Z=0),Scale3D=(X=1,Y=1,Z=1))
    Pendulum=(bEnabled=false,PendulumStiffness=50,PendulumGravity=(X=-900,Y=-500,Z=0),PendulumBlend=1,PendulumDrag=20,PendulumMinimum=0,PendulumMaximum=0.8,PendulumEase=SineEaseInOut,UnwindAxis=(X=0,Y=0,Z=0),UnwindMinimum=0,UnwindMaximum=0)
    Reach=(bEnabled=false,ReachTarget=(X=0,Y=0,Z=0),ReachAxis=(X=1,Y=1,Z=1),ReachMinimum=0,ReachMaximum=0,ReachEase=SineEaseInOut)
    Speed=(X=0,Y=-1.3,Z=0.2)
    Wave=(bEnabled=True,WaveFrequency=(X=0,Y=2.5,Z=2.5),WaveAmplitude=(X=0,Y=-1.5,Z=0.1),WaveOffset=(X=0,Y=1,Z=1),WaveNoise=(X=0,Y=0,Z=0),WaveMinimum=0.1,WaveMaximum=0.6,WaveEase=SineEaseInOut)
    WaveCurve="(EditorCurveData=(Keys=((InterpMode=RCIM_Linear,TangentMode=RCTM_Auto,TangentWeightMode=RCTWM_WeightedNone,Time=0,Value=0.2,ArriveTangent=0,ArriveTangentWeight=0,LeaveTangent=0,LeaveTangentWeight=0),(InterpMode=RCIM_Linear,TangentMode=RCTM_Auto,TangentWeig ...(truncated)"
    bDrawDebug=True
    Speed.Y=&RigVMFunction_MathDoubleMul(A=-1.3, B=RigVMVariableNode() {variable=Skydive_WaveSpeed}.Value).Result
   )
   > &RigUnit_ChainHarmonicsPerItem(
    ChainRoot=(Type=Bone,Name="dyn_Ribbon_r_in_01")
    DrawWorldOffset=(Rotation=(X=0,Y=0,Z=0,W=1),Translation=(X=0,Y=0,Z=0),Scale3D=(X=1,Y=1,Z=1))
    Pendulum=(bEnabled=false,PendulumStiffness=50,PendulumGravity=(X=-900,Y=-500,Z=0),PendulumBlend=1,PendulumDrag=20,PendulumMinimum=0,PendulumMaximum=0.8,PendulumEase=SineEaseInOut,UnwindAxis=(X=0,Y=0,Z=0),UnwindMinimum=0,UnwindMaximum=0)
    Reach=(bEnabled=false,ReachTarget=(X=0,Y=0,Z=0),ReachAxis=(X=1,Y=1,Z=1),ReachMinimum=0,ReachMaximum=0,ReachEase=SineEaseInOut)
    Speed=(X=0,Y=1.4,Z=0.2)
    Wave=(bEnabled=True,WaveFrequency=(X=0,Y=2,Z=2),WaveAmplitude=(X=0,Y=-1.6,Z=0.2),WaveOffset=(X=0,Y=1,Z=1),WaveNoise=(X=0,Y=0,Z=0),WaveMinimum=0.1,WaveMaximum=0.6,WaveEase=SineEaseInOut)
    WaveCurve="(EditorCurveData=(Keys=((InterpMode=RCIM_Linear,TangentMode=RCTM_Auto,TangentWeightMode=RCTWM_WeightedNone,Time=0,Value=0.2,ArriveTangent=0,ArriveTangentWeight=0,LeaveTangent=0,LeaveTangentWeight=0),(InterpMode=RCIM_Linear,TangentMode=RCTM_Auto,TangentWeig ...(truncated)"
    bDrawDebug=True
    Speed.Y=&RigVMFunction_MathDoubleMul(A=1.4, B=RigVMVariableNode() {variable=Skydive_WaveSpeed}.Value).Result
   )
   > &RigUnit_ChainHarmonicsPerItem(
    ChainRoot=(Type=Bone,Name="dyn_Ribbon_l_dw_01")
    DrawWorldOffset=(Rotation=(X=0,Y=0,Z=0,W=1),Translation=(X=0,Y=0,Z=0),Scale3D=(X=1,Y=1,Z=1))
    Pendulum=(bEnabled=false,PendulumStiffness=50,PendulumGravity=(X=-900,Y=-500,Z=0),PendulumBlend=1,PendulumDrag=20,PendulumMinimum=0,PendulumMaximum=0.8,PendulumEase=SineEaseInOut,UnwindAxis=(X=0,Y=0,Z=0),UnwindMinimum=0,UnwindMaximum=0)
    Reach=(bEnabled=false,ReachTarget=(X=0,Y=0,Z=0),ReachAxis=(X=1,Y=1,Z=1),ReachMinimum=0,ReachMaximum=0,ReachEase=SineEaseInOut)
    Speed=(X=0,Y=0,Z=1.5)
    Wave=(bEnabled=True,WaveFrequency=(X=0,Y=0,Z=6),WaveAmplitude=(X=0,Y=0,Z=-2.4),WaveOffset=(X=0,Y=0,Z=1),WaveNoise=(X=0,Y=0,Z=0),WaveMinimum=0.1,WaveMaximum=0.6,WaveEase=SineEaseInOut)
    WaveCurve="(EditorCurveData=(Keys=((InterpMode=RCIM_Linear,TangentMode=RCTM_Auto,TangentWeightMode=RCTWM_WeightedNone,Time=0,Value=0.2,ArriveTangent=0,ArriveTangentWeight=0,LeaveTangent=0,LeaveTangentWeight=0),(InterpMode=RCIM_Linear,TangentMode=RCTM_Auto,TangentWeig ...(truncated)"
    bDrawDebug=True
    Speed.Z=&RigVMFunction_MathDoubleMul(A=1.3, B=RigVMVariableNode() {variable=Skydive_WaveSpeed}.Value).Result
   )
   > &RigUnit_ChainHarmonicsPerItem(
    ChainRoot=(Type=Bone,Name="dyn_Ribbon_r_dw_01")
    DrawWorldOffset=(Rotation=(X=0,Y=0,Z=0,W=1),Translation=(X=0,Y=0,Z=0),Scale3D=(X=1,Y=1,Z=1))
    Pendulum=(bEnabled=false,PendulumStiffness=50,PendulumGravity=(X=-900,Y=-500,Z=0),PendulumBlend=1,PendulumDrag=20,PendulumMinimum=0,PendulumMaximum=0.8,PendulumEase=SineEaseInOut,UnwindAxis=(X=0,Y=0,Z=0),UnwindMinimum=0,UnwindMaximum=0)
    Reach=(bEnabled=false,ReachTarget=(X=0,Y=0,Z=0),ReachAxis=(X=1,Y=1,Z=1),ReachMinimum=0,ReachMaximum=0,ReachEase=SineEaseInOut)
    Speed=(X=0,Y=0,Z=1)
    Wave=(bEnabled=True,WaveFrequency=(X=0,Y=0,Z=7),WaveAmplitude=(X=0,Y=0,Z=-2.4),WaveOffset=(X=0,Y=0,Z=1),WaveNoise=(X=0,Y=0,Z=0),WaveMinimum=0.1,WaveMaximum=0.6,WaveEase=SineEaseInOut)
    WaveCurve="(EditorCurveData=(Keys=((InterpMode=RCIM_Linear,TangentMode=RCTM_Auto,TangentWeightMode=RCTWM_WeightedNone,Time=0,Value=0.2,ArriveTangent=0,ArriveTangentWeight=0,LeaveTangent=0,LeaveTangentWeight=0),(InterpMode=RCIM_Linear,TangentMode=RCTM_Auto,TangentWeig ...(truncated)"
    bDrawDebug=True
    Speed.Z=&RigVMFunction_MathDoubleMul(A=1, B=RigVMVariableNode() {variable=Skydive_WaveSpeed}.Value).Result
   )
   > &RigUnit_ChainHarmonicsPerItem(
    ChainRoot=(Type=Bone,Name="dyn_Ribbon_l_mid_01")
    DrawWorldOffset=(Rotation=(X=0,Y=0,Z=0,W=1),Translation=(X=0,Y=0,Z=0),Scale3D=(X=1,Y=1,Z=1))
    Pendulum=(bEnabled=false,PendulumStiffness=50,PendulumGravity=(X=-900,Y=-500,Z=0),PendulumBlend=1,PendulumDrag=20,PendulumMinimum=0,PendulumMaximum=0.8,PendulumEase=SineEaseInOut,UnwindAxis=(X=0,Y=0,Z=0),UnwindMinimum=0,UnwindMaximum=0)
    Reach=(bEnabled=false,ReachTarget=(X=0,Y=0,Z=0),ReachAxis=(X=1,Y=1,Z=1),ReachMinimum=0,ReachMaximum=0,ReachEase=SineEaseInOut)
    Speed=(X=0,Y=0,Z=1.5)
    Wave=(bEnabled=True,WaveFrequency=(X=0,Y=0,Z=6),WaveAmplitude=(X=0,Y=-2,Z=0),WaveOffset=(X=0,Y=0,Z=1),WaveNoise=(X=0,Y=0,Z=0),WaveMinimum=0.1,WaveMaximum=0.6,WaveEase=SineEaseInOut)
    WaveCurve="(EditorCurveData=(Keys=((InterpMode=RCIM_Linear,TangentMode=RCTM_Auto,TangentWeightMode=RCTWM_WeightedNone,Time=0,Value=0.2,ArriveTangent=0,ArriveTangentWeight=0,LeaveTangent=0,LeaveTangentWeight=0),(InterpMode=RCIM_Linear,TangentMode=RCTM_Auto,TangentWeig ...(truncated)"
    bDrawDebug=True
    Speed.Z=&RigVMFunction_MathDoubleMul(A=1.5, B=RigVMVariableNode() {variable=Skydive_WaveSpeed}.Value).Result
   )
   > &RigUnit_ChainHarmonicsPerItem(
    ChainRoot=(Type=Bone,Name="dyn_Ribbon_r_mid_01")
    DrawWorldOffset=(Rotation=(X=0,Y=0,Z=0,W=1),Translation=(X=0,Y=0,Z=0),Scale3D=(X=1,Y=1,Z=1))
    Pendulum=(bEnabled=false,PendulumStiffness=50,PendulumGravity=(X=-900,Y=-500,Z=0),PendulumBlend=1,PendulumDrag=20,PendulumMinimum=0,PendulumMaximum=0.8,PendulumEase=SineEaseInOut,UnwindAxis=(X=0,Y=0,Z=0),UnwindMinimum=0,UnwindMaximum=0)
    Reach=(bEnabled=false,ReachTarget=(X=0,Y=0,Z=0),ReachAxis=(X=1,Y=1,Z=1),ReachMinimum=0,ReachMaximum=0,ReachEase=SineEaseInOut)
    Speed=(X=0,Y=0,Z=1)
    Wave=(bEnabled=True,WaveFrequency=(X=0,Y=0,Z=7),WaveAmplitude=(X=0,Y=0,Z=-3),WaveOffset=(X=0,Y=0,Z=1),WaveNoise=(X=0,Y=0,Z=0),WaveMinimum=0.1,WaveMaximum=0.6,WaveEase=SineEaseInOut)
    WaveCurve="(EditorCurveData=(Keys=((InterpMode=RCIM_Linear,TangentMode=RCTM_Auto,TangentWeightMode=RCTWM_WeightedNone,Time=0,Value=0.2,ArriveTangent=0,ArriveTangentWeight=0,LeaveTangent=0,LeaveTangentWeight=0),(InterpMode=RCIM_Linear,TangentMode=RCTM_Auto,TangentWeig ...(truncated)"
    bDrawDebug=True
    Speed.Z=&RigVMFunction_MathDoubleMul(A=1, B=RigVMVariableNode() {variable=Skydive_WaveSpeed}.Value).Result
   )
  ]}
]
hierarchy [
 root {loc=<0 0 0>, rot=<0 0 0 1>}
 pack {parent=root, loc=<0 -0.7316 0>, rot=<0 0 0 1>}
 base_Ribbon_l_01 {parent=pack, loc=<1.5097 -1.0639 -8.3555>, rot=<-0.6353 -0.1441 0.639 -0.409>}
 dyn_Ribbon_l_01 {parent=base_Ribbon_l_01, loc=<0 0 0>, rot=<0 0 0 1>}
 dyn_Ribbon_l_02 {parent=dyn_Ribbon_l_01, loc=<7.9485 -0.2104 0.1024>, rot=<0 0 -0.0279 -0.9996>}
 dyn_Ribbon_l_03 {parent=dyn_Ribbon_l_02, loc=<13.239 0.2385 0.2685>, rot=<0 0 -0.0797 -0.9968>}
 dyn_Ribbon_l_04 {parent=dyn_Ribbon_l_03, loc=<13.1207 0.7323 -0.8436>, rot=<0.0055 -0.0734 -0.0748 -0.9945>}
 dyn_Ribbon_l_05 {parent=dyn_Ribbon_l_04, loc=<13.4513 0.034 -0.7765>, rot=<0.0028 -0.0516 0.0544 -0.9972>}
 base_Ribbon_r_01 {parent=pack, loc=<-1.5097 -1.0639 -8.3554>, rot=<-0.1441 -0.6353 0.409 -0.639>}
 dyn_Ribbon_r_01 {parent=base_Ribbon_r_01, loc=<0 0 0>, rot=<0 0 0 1>}
 dyn_Ribbon_r_02 {parent=dyn_Ribbon_r_01, loc=<7.9485 0.2104 0.1024>, rot=<0 0 0.0279 -0.9996>}
 dyn_Ribbon_r_03 {parent=dyn_Ribbon_r_02, loc=<13.239 -0.2385 0.2685>, rot=<0 0 0.0797 -0.9968>}
 dyn_Ribbon_r_04 {parent=dyn_Ribbon_r_03, loc=<13.1208 -0.7323 -0.8436>, rot=<-0.0055 -0.0734 0.0748 -0.9945>}
 dyn_Ribbon_r_05 {parent=dyn_Ribbon_r_04, loc=<13.4512 -0.0341 -0.7765>, rot=<-0.0028 -0.0516 -0.0544 -0.9972>}
 base_Ribbon_l_in_01 {parent=pack, loc=<3.5904 -0.8279 0.6338>, rot=<-0.6491 0.3539 0.1493 -0.6566>}
 dyn_Ribbon_l_in_01 {parent=base_Ribbon_l_in_01, loc=<0 0 0>, rot=<0 0 0 1>}
 dyn_Ribbon_l_in_02 {parent=dyn_Ribbon_l_in_01, loc=<10.4597 1.5988 -0.4433>, rot=<0.0715 0.0633 -0.0424 -0.9945>}
 dyn_Ribbon_l_in_03 {parent=dyn_Ribbon_l_in_02, loc=<14.4277 1.4052 0.5839>, rot=<0.055 0.0431 0.7065 -0.7043>}
 dyn_Ribbon_l_in_04 {parent=dyn_Ribbon_l_in_03, loc=<-2.6509 14.566 1.8773>, rot=<-0.0413 -0.0011 -0.0788 -0.996>}
 base_Ribbon_r_in_01 {parent=pack, loc=<-3.5904 -0.8279 0.6338>, rot=<0.1493 -0.6566 -0.6491 0.3539>}
 dyn_Ribbon_r_in_01 {parent=base_Ribbon_r_in_01, loc=<0 0 0>, rot=<0 0 0 1>}
 dyn_Ribbon_r_in_02 {parent=dyn_Ribbon_r_in_01, loc=<10.4597 1.5987 0.4434>, rot=<-0.0715 -0.0633 -0.0424 -0.9945>}
 dyn_Ribbon_r_in_03 {parent=dyn_Ribbon_r_in_02, loc=<14.4277 1.4052 -0.5839>, rot=<-0.055 -0.0431 0.7065 -0.7043>}
 dyn_Ribbon_r_in_04 {parent=dyn_Ribbon_r_in_03, loc=<-2.651 14.566 -1.8774>, rot=<0.0413 0.0011 -0.0788 -0.996>}
 base_Ribbon_l_dw_01 {parent=pack, loc=<3.9652 -0.9445 -8.2935>, rot=<0.6927 0.2095 -0.5006 0.475>}
 dyn_Ribbon_l_dw_01 {parent=base_Ribbon_l_dw_01, loc=<0 0 0>, rot=<0 0 0 1>}
 dyn_Ribbon_l_dw_02 {parent=dyn_Ribbon_l_dw_01, loc=<7.5187 0.272 1.0045>, rot=<-0.0483 0.0688 -0.147 -0.9856>}
 dyn_Ribbon_l_dw_03 {parent=dyn_Ribbon_l_dw_02, loc=<10.4656 0.2644 0.5649>, rot=<0.0165 0 -0.1264 -0.9918>}
 dyn_Ribbon_l_dw_04 {parent=dyn_Ribbon_l_dw_03, loc=<10.7384 0.4391 1.3991>, rot=<-0.0044 0.0675 -0.0224 -0.9975>}
 dyn_Ribbon_l_dw_05 {parent=dyn_Ribbon_l_dw_04, loc=<7.01 -0.0057 0.0369>, rot=<0.0231 -0.0703 0.0647 -0.9952>}
 base_Ribbon_r_dw_01 {parent=pack, loc=<-3.9652 -0.9445 -8.2935>, rot=<-0.5006 0.475 0.6927 0.2095>}
 dyn_Ribbon_r_dw_01 {parent=base_Ribbon_r_dw_01, loc=<0 0 0>, rot=<0 0 0 -1>}
 dyn_Ribbon_r_dw_02 {parent=dyn_Ribbon_r_dw_01, loc=<7.5187 0.272 -1.0045>, rot=<0.0483 -0.0688 -0.147 -0.9856>}
 dyn_Ribbon_r_dw_03 {parent=dyn_Ribbon_r_dw_02, loc=<10.4656 0.2645 -0.5648>, rot=<-0.0165 0 -0.1264 -0.9918>}
 dyn_Ribbon_r_dw_04 {parent=dyn_Ribbon_r_dw_03, loc=<10.7383 0.4391 -1.3991>, rot=<0.0044 -0.0675 -0.0224 -0.9975>}
 dyn_Ribbon_r_dw_05 {parent=dyn_Ribbon_r_dw_04, loc=<7.01 -0.0057 -0.0369>, rot=<-0.0231 0.0703 0.0647 -0.9952>}
 base_Ribbon_l_mid_02 {parent=pack, loc=<3.8311 -1.7567 -0.2773>, rot=<-0.6773 0.2378 0.286 -0.6347>}
 dyn_Ribbon_l_mid_01 {parent=base_Ribbon_l_mid_02, loc=<0 0 0>, rot=<0 0 0 1>}
 dyn_Ribbon_l_mid_02 {parent=dyn_Ribbon_l_mid_01, loc=<6.397 0 0.0225>, rot=<0 0.0645 0 -0.9979>}
 dyn_Ribbon_l_mid_03 {parent=dyn_Ribbon_l_mid_02, loc=<6.5746 0.0472 0.8159>, rot=<0.0669 0.0233 0.0336 -0.9969>}
 base_Ribbon_r_mid_01 {parent=pack, loc=<-3.8311 -1.7567 -0.2773>, rot=<-0.286 0.6347 0.6773 -0.2378>}
 dyn_Ribbon_r_mid_01 {parent=base_Ribbon_r_mid_01, loc=<0 0 0>, rot=<0 0 0 1>}
 dyn_Ribbon_r_mid_02 {parent=dyn_Ribbon_r_mid_01, loc=<6.397 0 -0.0225>, rot=<0 -0.0645 0 -0.9979>}
 dyn_Ribbon_r_mid_03 {parent=dyn_Ribbon_r_mid_02, loc=<6.5746 0.0472 -0.8159>, rot=<-0.0669 -0.0233 0.0336 -0.9969>}
]
variables [
 Skydive {type=bool, default=False}
 Skydive_WaveSpeed {type=double, default=0}
]
