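# Animation Blueprint for the MajorSpeech backpack cosmetic, adding secondary motion to its ribbons and repositioning the pack.
# Chained Anim Dynamics nodes simulate the ribbon bones, then additive bone transforms and a Control Rig refine the pose.
# Skydiving halves outer ribbon dynamics, swaps sim-space settings and raises Control Rig wave speed.
# Each update also maps spine bend to pack offset and rotation, and caches falling, gender and body-type flags.
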AnimBlueprint /BRCosmetics/Accessories/FORT_Backpacks/Backpack_F_MED_MajorSpeech/Meshes/F_MED_MajorSpeech_Pack_AnimBP
target_skeleton=&/F_MED_MajorSpeech_Pack_Skeleton
graphs [
 AnimGraph {nodes [
   n1 = Get Skydive_alpha() -> (Skydive_alpha:double)
   n2 = Get Skydive_alpha()
   n3 = Anim Dynamics - Bone: dyn_Ribbon_l_01(ComponentPose=Local To Component(LocalPose=Input Pose()), Alpha=@n2, ExternalForce="0,0,0") {
    angularDampingOverride=1
    angularSpringConstant=7000
    bAngularSpring=true
    bOverrideAngularDamping=true
    bUseGravityOverride=true
    gravityScale=0.35
    simulationSpace=RootRelative
    boundBone {boneName=dyn_Ribbon_l_01}
    componentAppliedLinearAccClamp=<4000 4000 4000>
    componentLinearAccScale=<0.25 0.25 0.25>
    componentLinearVelScale=<0.25 0.25 0.25>
    gravityOverride {z=-880}
    physicsBodyDefinitions [{
      boundBone {boneName=dyn_Ribbon_l_01}
      boxExtents=<20 26 26>
      constraintSetup {angularLimitsMax=<0 4 15>, angularLimitsMin=<0 -18 0>, angularTarget=<1 0 0>}
      localJointOffset=<4.1 0 0>
     }]
   }
   n4 = Get Sim_space_settings() -> (Sim_space_settings:Anim Phys Sim Space Settings Structure)
   n5 = Anim Dynamics - Bone: dyn_Ribbon_l_02!(ComponentPose=@n3, Alpha=@n2, SimSpaceSettings:Anim Phys Sim Space Settings Structure=@n4, ExternalForce="0,0,0") -> (Pose) {
    angularDampingOverride=1
    angularSpringConstant=6000
    bAngularSpring=true
    bOverrideAngularDamping=true
    bUseGravityOverride=true
    gravityScale=0.35
    simulationSpace=RootRelative
    boundBone {boneName=dyn_Ribbon_l_02}
    componentAppliedLinearAccClamp=<4000 4000 4000>
    componentLinearAccScale=<0.25 0.25 0.25>
    componentLinearVelScale=<0.25 0.25 0.25>
    gravityOverride {z=-890}
    physicsBodyDefinitions [{
      boundBone {boneName=dyn_Ribbon_l_02}
      boxExtents=<20 24 24>
      constraintSetup {angularLimitsMax=<0 6 20>, angularLimitsMin=<0 -20 -5>, angularTarget=<1 0 0>}
      localJointOffset=<4.15 0 0>
     }]
   }
   n6 = Anim Dynamics - Bone: dyn_Ribbon_l_03!(ComponentPose=@n5, Alpha=@n2, SimSpaceSettings:Anim Phys Sim Space Settings Structure=@n4, ExternalForce="0,0,0") -> (Pose) {
    angularDampingOverride=1
    angularSpringConstant=6000
    bAngularSpring=true
    bOverrideAngularDamping=true
    bUseGravityOverride=true
    gravityScale=0.35
    simulationSpace=RootRelative
    boundBone {boneName=dyn_Ribbon_l_03}
    componentAppliedLinearAccClamp=<4000 4000 4000>
    componentLinearAccScale=<0.25 0.25 0.25>
    componentLinearVelScale=<0.25 0.25 0.25>
    gravityOverride {z=-890}
    physicsBodyDefinitions [{
      boundBone {boneName=dyn_Ribbon_l_03}
      boxExtents=<20 22 22>
      constraintSetup {angularLimitsMax=<0 6 20>, angularLimitsMin=<0 -22 -6>, angularTarget=<1 0 0>}
      localJointOffset=<4.3 0 0>
     }]
   }
   n7 = Anim Dynamics - Bone: dyn_Ribbon_l_04!(ComponentPose=@n6, Alpha=@n2, SimSpaceSettings:Anim Phys Sim Space Settings Structure=@n4, ExternalForce="0,0,0") -> (Pose) {
    angularDampingOverride=1
    angularSpringConstant=6000
    bAngularSpring=true
    bOverrideAngularDamping=true
    bUseGravityOverride=true
    gravityScale=0.35
    simulationSpace=RootRelative
    boundBone {boneName=dyn_Ribbon_l_04}
    componentAppliedLinearAccClamp=<4000 4000 4000>
    componentLinearAccScale=<0.25 0.25 0.25>
    componentLinearVelScale=<0.25 0.25 0.25>
    gravityOverride {z=-890}
    physicsBodyDefinitions [{
      boundBone {boneName=dyn_Ribbon_l_04}
      boxExtents=<20 22 22>
      constraintSetup {angularLimitsMax=<0 6 20>, angularLimitsMin=<0 -22 -6>, angularTarget=<1 0 0>}
      localJointOffset=<4.3 0 0>
     }]
   }
   n8 = Anim Dynamics - Bone: dyn_Ribbon_l_04(ComponentPose=@n7, Alpha=@n2, SimSpaceSettings=@n4, ExternalForce="0,0,0") {
    angularDampingOverride=1
    angularSpringConstant=6000
    bAngularSpring=true
    bOverrideAngularDamping=true
    bUseGravityOverride=true
    gravityScale=0.35
    simulationSpace=RootRelative
    boundBone {boneName=dyn_Ribbon_l_04}
    componentAppliedLinearAccClamp=<4000 4000 4000>
    componentLinearAccScale=<0.25 0.25 0.25>
    componentLinearVelScale=<0.25 0.25 0.25>
    gravityOverride {z=-890}
    physicsBodyDefinitions [{
      boundBone {boneName=dyn_Ribbon_l_04}
      boxExtents=<20 22 22>
      constraintSetup {angularLimitsMax=<0 6 20>, angularLimitsMin=<0 -22 -6>, angularTarget=<1 0 0>}
      localJointOffset=<4.3 0 0>
     }]
   }
   n9 = Anim Dynamics - Bone: dyn_Ribbon_r_01(ComponentPose=@n8, Alpha=@n1, ExternalForce="0,0,0") {
    angularDampingOverride=1
    angularSpringConstant=7000
    bAngularSpring=true
    bOverrideAngularDamping=true
    bUseGravityOverride=true
    gravityScale=0.35
    simulationSpace=RootRelative
    boundBone {boneName=dyn_Ribbon_r_01}
    componentAppliedLinearAccClamp=<4000 4000 4000>
    componentLinearAccScale=<0.25 0.25 0.25>
    componentLinearVelScale=<0.25 0.25 0.25>
    gravityOverride {z=-880}
    physicsBodyDefinitions [{
      boundBone {boneName=dyn_Ribbon_r_01}
      boxExtents=<20 28 28>
      constraintSetup {angularLimitsMax=<0 16 8>, angularLimitsMin=<0 -2 -25>, angularTarget=<1 0 0>}
      localJointOffset=<4.1 0 0>
     }]
   }
   n10 = Get Sim_space_settings()
   n11 = Anim Dynamics - Bone: dyn_Ribbon_r_02!(ComponentPose=@n9, Alpha=@n1, SimSpaceSettings:Anim Phys Sim Space Settings Structure=@n10, ExternalForce="0,0,0") -> (Pose) {
    angularDampingOverride=1
    angularSpringConstant=6000
    bAngularSpring=true
    bOverrideAngularDamping=true
    bUseGravityOverride=true
    gravityScale=0.35
    simulationSpace=RootRelative
    boundBone {boneName=dyn_Ribbon_r_02}
    componentAppliedLinearAccClamp=<4000 4000 4000>
    componentLinearAccScale=<0.25 0.25 0.25>
    componentLinearVelScale=<0.25 0.25 0.25>
    gravityOverride {z=-890}
    physicsBodyDefinitions [{
      boundBone {boneName=dyn_Ribbon_r_02}
      boxExtents=<20 25 25>
      constraintSetup {angularLimitsMax=<0 19 10>, angularLimitsMin=<0 -5 -18>, angularTarget=<1 0 0>}
      localJointOffset=<4.2 0 0>
     }]
   }
   n12 = Anim Dynamics - Bone: dyn_Ribbon_r_03!(ComponentPose=@n11, Alpha=@n1, SimSpaceSettings:Anim Phys Sim Space Settings Structure=@n10, ExternalForce="0,0,0") -> (Pose) {
    angularDampingOverride=1
    angularSpringConstant=6000
    bAngularSpring=true
    bOverrideAngularDamping=true
    bUseGravityOverride=true
    gravityScale=0.35
    simulationSpace=RootRelative
    boundBone {boneName=dyn_Ribbon_r_03}
    componentAppliedLinearAccClamp=<4000 4000 4000>
    componentLinearAccScale=<0.25 0.25 0.25>
    componentLinearVelScale=<0.25 0.25 0.25>
    gravityOverride {z=-890}
    physicsBodyDefinitions [{
      boundBone {boneName=dyn_Ribbon_r_03}
      boxExtents=<20 23 23>
      constraintSetup {angularLimitsMax=<0 23 15>, angularLimitsMin=<0 -5 -20>, angularTarget=<1 0 0>}
      localJointOffset=<4.2 0 0>
     }]
   }
   n13 = Anim Dynamics - Bone: dyn_Ribbon_r_04!(ComponentPose=@n12, Alpha=@n1, SimSpaceSettings:Anim Phys Sim Space Settings Structure=@n10, ExternalForce="0,0,0") -> (Pose) {
    angularDampingOverride=1
    angularSpringConstant=6000
    bAngularSpring=true
    bOverrideAngularDamping=true
    bUseGravityOverride=true
    gravityScale=0.35
    simulationSpace=RootRelative
    boundBone {boneName=dyn_Ribbon_r_04}
    componentAppliedLinearAccClamp=<4000 4000 4000>
    componentLinearAccScale=<0.25 0.25 0.25>
    componentLinearVelScale=<0.25 0.25 0.25>
    gravityOverride {z=-890}
    physicsBodyDefinitions [{
      boundBone {boneName=dyn_Ribbon_r_04}
      boxExtents=<20 23 23>
      constraintSetup {angularLimitsMax=<0 23 15>, angularLimitsMin=<0 -5 -20>, angularTarget=<1 0 0>}
      localJointOffset=<4.2 0 0>
     }]
   }
   n14 = Anim Dynamics - Bone: dyn_Ribbon_r_05!(ComponentPose=@n13, Alpha=@n1, SimSpaceSettings:Anim Phys Sim Space Settings Structure=@n10, ExternalForce="0,0,0") -> (Pose) {
    angularDampingOverride=1
    angularSpringConstant=6000
    bAngularSpring=true
    bOverrideAngularDamping=true
    bUseGravityOverride=true
    gravityScale=0.35
    simulationSpace=RootRelative
    boundBone {boneName=dyn_Ribbon_r_05}
    componentAppliedLinearAccClamp=<4000 4000 4000>
    componentLinearAccScale=<0.25 0.25 0.25>
    componentLinearVelScale=<0.25 0.25 0.25>
    gravityOverride {z=-890}
    physicsBodyDefinitions [{
      boundBone {boneName=dyn_Ribbon_r_05}
      boxExtents=<20 23 23>
      constraintSetup {angularLimitsMax=<0 23 15>, angularLimitsMin=<0 -5 -20>, angularTarget=<1 0 0>}
      localJointOffset=<4.2 0 0>
     }]
   }
   n15 = Anim Dynamics - Bone: dyn_Ribbon_l_in_01(ComponentPose=@n14, Alpha=1, ExternalForce="0,0,0") {
    angularDampingOverride=1
    angularSpringConstant=5000
    bAngularSpring=true
    bOverrideAngularDamping=true
    bUseGravityOverride=true
    gravityScale=0.35
    simulationSpace=RootRelative
    boundBone {boneName=dyn_Ribbon_l_in_01}
    componentAppliedLinearAccClamp=<3000 3000 3000>
    componentLinearAccScale=<0.2 0.2 0.2>
    componentLinearVelScale=<0.2 0.2 0.2>
    gravityOverride {z=-780}
    physicsBodyDefinitions [{
      boundBone {boneName=dyn_Ribbon_l_in_01}
      boxExtents=<20 30 30>
      constraintSetup {angularLimitsMax=<0 9 9>, angularLimitsMin=<0 -9 -9>, angularTarget=<1 0 0>}
      localJointOffset=<2.7 0 0>
     }]
   }
   n16 = Anim Dynamics - Bone: dyn_Ribbon_l_in_02(ComponentPose=@n15, Alpha=1, ExternalForce="0,0,0") {
    angularDampingOverride=1
    angularSpringConstant=5000
    bAngularSpring=true
    bOverrideAngularDamping=true
    bUseGravityOverride=true
    gravityScale=0.35
    simulationSpace=RootRelative
    boundBone {boneName=dyn_Ribbon_l_in_02}
    componentAppliedLinearAccClamp=<3000 3000 3000>
    componentLinearAccScale=<0.2 0.2 0.2>
    componentLinearVelScale=<0.2 0.2 0.2>
    gravityOverride {z=-780}
    physicsBodyDefinitions [{
      boundBone {boneName=dyn_Ribbon_l_in_02}
      boxExtents=<20 30 30>
      constraintSetup {angularLimitsMax=<0 9 9>, angularLimitsMin=<0 -9 -9>, angularTarget=<1 0 0>}
      localJointOffset=<2.7 0 0>
     }]
   }
   n17 = Anim Dynamics - Bone: dyn_Ribbon_r_in_01(ComponentPose=@n16, Alpha=1, ExternalForce="0,0,0") {
    angularDampingOverride=1
    angularSpringConstant=5000
    bAngularSpring=true
    bOverrideAngularDamping=true
    bUseGravityOverride=true
    gravityScale=0.35
    simulationSpace=RootRelative
    boundBone {boneName=dyn_Ribbon_r_in_01}
    componentAppliedLinearAccClamp=<3000 3000 3000>
    componentLinearAccScale=<0.2 0.2 0.2>
    componentLinearVelScale=<0.2 0.2 0.2>
    gravityOverride {z=-780}
    physicsBodyDefinitions [{
      boundBone {boneName=dyn_Ribbon_r_in_01}
      boxExtents=<20 30 30>
      constraintSetup {angularLimitsMax=<0 9 9>, angularLimitsMin=<0 -9 -9>, angularTarget=<1 0 0>}
      localJointOffset=<2.7 0 0>
     }]
   }
   n18 = Anim Dynamics - Bone: dyn_Ribbon_r_in_02(ComponentPose=@n17, Alpha=1, ExternalForce="0,0,0") {
    angularDampingOverride=1
    angularSpringConstant=5000
    bAngularSpring=true
    bOverrideAngularDamping=true
    bUseGravityOverride=true
    gravityScale=0.35
    simulationSpace=RootRelative
    boundBone {boneName=dyn_Ribbon_r_in_02}
    componentAppliedLinearAccClamp=<3000 3000 3000>
    componentLinearAccScale=<0.2 0.2 0.2>
    componentLinearVelScale=<0.2 0.2 0.2>
    gravityOverride {z=-780}
    physicsBodyDefinitions [{
      boundBone {boneName=dyn_Ribbon_r_in_02}
      boxExtents=<20 30 30>
      constraintSetup {angularLimitsMax=<0 9 9>, angularLimitsMin=<0 -9 -9>, angularTarget=<1 0 0>}
      localJointOffset=<2.7 0 0>
     }]
   }
   n19 = Anim Dynamics - Bone: dyn_Ribbon_l_dw_01(ComponentPose=@n18, Alpha=1, ExternalForce="0,0,0") {
    angularDampingOverride=1
    angularSpringConstant=7000
    bAngularSpring=true
    bOverrideAngularDamping=true
    bUseGravityOverride=true
    gravityScale=0.35
    simulationSpace=RootRelative
    boundBone {boneName=dyn_Ribbon_l_dw_01}
    componentAppliedLinearAccClamp=<3000 3000 3000>
    componentLinearAccScale=<0.2 0.2 0.2>
    componentLinearVelScale=<0.2 0.2 0.2>
    gravityOverride {z=-780}
    physicsBodyDefinitions [{
      boundBone {boneName=dyn_Ribbon_l_dw_01}
      boxExtents=<20 30 30>
      constraintSetup {angularLimitsMax=<0 5 10>, angularLimitsMin=<0 -15 -10>, angularTarget=<1 0 0>}
      localJointOffset=<2.7 0 0>
     }]
   }
   n20 = Anim Dynamics - Bone: dyn_Ribbon_l_dw_02(ComponentPose=@n19, Alpha=1, ExternalForce="0,0,0") {
    angularDampingOverride=1
    angularSpringConstant=7000
    bAngularSpring=true
    bOverrideAngularDamping=true
    bUseGravityOverride=true
    gravityScale=0.35
    simulationSpace=RootRelative
    boundBone {boneName=dyn_Ribbon_l_dw_02}
    componentAppliedLinearAccClamp=<3000 3000 3000>
    componentLinearAccScale=<0.2 0.2 0.2>
    componentLinearVelScale=<0.2 0.2 0.2>
    gravityOverride {z=-780}
    physicsBodyDefinitions [{
      boundBone {boneName=dyn_Ribbon_l_dw_02}
      boxExtents=<20 30 30>
      constraintSetup {angularLimitsMax=<0 9 10>, angularLimitsMin=<0 -15 -10>, angularTarget=<1 0 0>}
      localJointOffset=<2.7 0 0>
     }]
   }
   n21 = Anim Dynamics - Bone: dyn_Ribbon_r_dw_01(ComponentPose=@n20, Alpha=1, ExternalForce="0,0,0") {
    angularDampingOverride=1
    angularSpringConstant=7000
    bAngularSpring=true
    bOverrideAngularDamping=true
    bUseGravityOverride=true
    gravityScale=0.35
    simulationSpace=RootRelative
    boundBone {boneName=dyn_Ribbon_r_dw_01}
    componentAppliedLinearAccClamp=<3000 3000 3000>
    componentLinearAccScale=<0.2 0.2 0.2>
    componentLinearVelScale=<0.2 0.2 0.2>
    gravityOverride {z=-780}
    physicsBodyDefinitions [{
      boundBone {boneName=dyn_Ribbon_r_dw_01}
      boxExtents=<20 30 30>
      constraintSetup {angularLimitsMax=<0 5 10>, angularLimitsMin=<0 -15 -10>, angularTarget=<1 0 0>}
      localJointOffset=<2.7 0 0>
     }]
   }
   n22 = Anim Dynamics - Bone: dyn_Ribbon_r_dw_02(ComponentPose=@n21, Alpha=1, ExternalForce="0,0,0") {
    angularDampingOverride=1
    angularSpringConstant=7000
    bAngularSpring=true
    bOverrideAngularDamping=true
    bUseGravityOverride=true
    gravityScale=0.35
    simulationSpace=RootRelative
    boundBone {boneName=dyn_Ribbon_r_dw_02}
    componentAppliedLinearAccClamp=<3000 3000 3000>
    componentLinearAccScale=<0.2 0.2 0.2>
    componentLinearVelScale=<0.2 0.2 0.2>
    gravityOverride {z=-780}
    physicsBodyDefinitions [{
      boundBone {boneName=dyn_Ribbon_r_dw_02}
      boxExtents=<20 30 30>
      constraintSetup {angularLimitsMax=<0 9 10>, angularLimitsMin=<0 -15 -10>, angularTarget=<1 0 0>}
      localJointOffset=<2.7 0 0>
     }]
   }
   n23 = Anim Dynamics - Bone: dyn_Ribbon_l_mid_01(ComponentPose=@n22, Alpha=1, ExternalForce="0,0,0") {
    angularDampingOverride=1
    angularSpringConstant=7000
    bAngularSpring=true
    bOverrideAngularDamping=true
    bUseGravityOverride=true
    gravityScale=0.35
    simulationSpace=RootRelative
    boundBone {boneName=dyn_Ribbon_l_mid_01}
    componentAppliedLinearAccClamp=<3000 3000 3000>
    componentLinearAccScale=<0.2 0.2 0.2>
    componentLinearVelScale=<0.2 0.2 0.2>
    gravityOverride {z=-780}
    physicsBodyDefinitions [{
      boundBone {boneName=dyn_Ribbon_l_mid_01}
      boxExtents=<20 30 30>
      constraintSetup {angularLimitsMax=<0 15 10>, angularLimitsMin=<0 -15 -10>, angularTarget=<1 0 0>}
      localJointOffset=<2.7 0 0>
     }]
   }
   n24 = Anim Dynamics - Bone: dyn_Ribbon_r_mid_01(ComponentPose=@n23, Alpha=1, ExternalForce="0,0,0") {
    angularDampingOverride=1
    angularSpringConstant=7000
    bAngularSpring=true
    bOverrideAngularDamping=true
    bUseGravityOverride=true
    gravityScale=0.35
    simulationSpace=RootRelative
    boundBone {boneName=dyn_Ribbon_r_mid_01}
    componentAppliedLinearAccClamp=<3000 3000 3000>
    componentLinearAccScale=<0.2 0.2 0.2>
    componentLinearVelScale=<0.2 0.2 0.2>
    gravityOverride {z=-780}
    physicsBodyDefinitions [{
      boundBone {boneName=dyn_Ribbon_r_mid_01}
      boxExtents=<20 30 30>
      constraintSetup {angularLimitsMax=<0 15 10>, angularLimitsMin=<0 -15 -10>, angularTarget=<1 0 0>}
      localJointOffset=<2.7 0 0>
     }]
   }
   n25 = Anim Dynamics - Bone: dyn_Ribbon_l_mid_02(ComponentPose=@n24, Alpha=1, ExternalForce="0,0,0") {
    angularDampingOverride=1
    angularSpringConstant=7000
    bAngularSpring=true
    bOverrideAngularDamping=true
    bUseGravityOverride=true
    gravityScale=0.35
    simulationSpace=RootRelative
    boundBone {boneName=dyn_Ribbon_l_mid_02}
    componentAppliedLinearAccClamp=<3000 3000 3000>
    componentLinearAccScale=<0.2 0.2 0.2>
    componentLinearVelScale=<0.2 0.2 0.2>
    gravityOverride {z=-780}
    physicsBodyDefinitions [{
      boundBone {boneName=dyn_Ribbon_l_mid_02}
      boxExtents=<20 30 30>
      constraintSetup {angularLimitsMax=<0 15 10>, angularLimitsMin=<0 -15 -10>, angularTarget=<1 0 0>}
      localJointOffset=<2.7 0 0>
     }]
   }
   n26 = Anim Dynamics - Bone: dyn_Ribbon_r_mid_02(ComponentPose=@n25, Alpha=1, ExternalForce="0,0,0") {
    angularDampingOverride=1
    angularSpringConstant=7000
    bAngularSpring=true
    bOverrideAngularDamping=true
    bUseGravityOverride=true
    gravityScale=0.35
    simulationSpace=RootRelative
    boundBone {boneName=dyn_Ribbon_r_mid_02}
    componentAppliedLinearAccClamp=<3000 3000 3000>
    componentLinearAccScale=<0.2 0.2 0.2>
    componentLinearVelScale=<0.2 0.2 0.2>
    gravityOverride {z=-780}
    physicsBodyDefinitions [{
      boundBone {boneName=dyn_Ribbon_r_mid_02}
      boxExtents=<20 30 30>
      constraintSetup {angularLimitsMax=<0 15 10>, angularLimitsMin=<0 -15 -10>, angularTarget=<1 0 0>}
      localJointOffset=<2.7 0 0>
     }]
   }
   n27 = "Transform (Modify) Bone - Bone: base_Ribbon_l_01"!(ComponentPose=@n26, bAlphaBoolEnabled:bool=True, Translation="0,0,0", Rotation="16.861,-1.969,-0.945", Scale="1,1,1") -> (Pose) {
    alphaInputType=Bool
    rotationMode=BMM_Additive
    rotationSpace=BCS_BoneSpace
    alphaBoolBlend {blendInTime=0.3, blendOutTime=0.3}
    boneToModify {boneName=base_Ribbon_l_01}
    rotation=<rot 16.861 -1.969 -0.945>
   }
   n28 = "Transform (Modify) Bone - Bone: base_Ribbon_r_01"!(ComponentPose=@n27, bAlphaBoolEnabled:bool=True, Translation="0,0,0", Rotation="15,-12,-1", Scale="1,1,1") -> (Pose) {
    alphaInputType=Bool
    rotationMode=BMM_Additive
    rotationSpace=BCS_BoneSpace
    alphaBoolBlend {blendInTime=0.3, blendOutTime=0.3}
    boneToModify {boneName=base_Ribbon_r_01}
    rotation=<rot 15 -12 -1>
   }
   n29 = "Transform (Modify) Bone - Bone: base_Ribbon_l_01"(ComponentPose=@n28, bAlphaBoolEnabled=True, Translation="0,0,0", Rotation="10,15,0", Scale="1,1,1") {
    alphaInputType=Bool
    rotationMode=BMM_Additive
    rotationSpace=BCS_BoneSpace
    alphaBoolBlend {blendInTime=0.3, blendOutTime=0.3}
    boneToModify {boneName=base_Ribbon_l_01}
    rotation {pitch=10, yaw=15}
   }
   n30 = "Transform (Modify) Bone - Bone: base_Ribbon_r_01"(ComponentPose=@n29, bAlphaBoolEnabled=True, Translation="0,0,0", Rotation="12,-15,0", Scale="1,1,1") {
    alphaInputType=Bool
    rotationMode=BMM_Additive
    rotationSpace=BCS_BoneSpace
    alphaBoolBlend {blendInTime=0.3, blendOutTime=0.3}
    boneToModify {boneName=base_Ribbon_r_01}
    rotation {pitch=12, yaw=-15}
   }
   n31 = "Transform (Modify) Bone - Bone: base_Ribbon_l_in_01"!(ComponentPose=@n30, bAlphaBoolEnabled:bool=True, Translation="0,0,0", Rotation="5,-17,-5", Scale="1,1,1") -> (Pose) {
    alphaInputType=Bool
    rotationMode=BMM_Additive
    rotationSpace=BCS_BoneSpace
    alphaBoolBlend {blendInTime=0.3, blendOutTime=0.3}
    boneToModify {boneName=base_Ribbon_l_in_01}
    rotation=<rot 5 -17 -5>
   }
   n32 = "Transform (Modify) Bone - Bone: base_Ribbon_r_in_01"!(ComponentPose=@n31, bAlphaBoolEnabled:bool=True, Translation="0,0,0", Rotation="-4,-18,5", Scale="1,1,1") -> (Pose) {
    alphaInputType=Bool
    rotationMode=BMM_Additive
    rotationSpace=BCS_BoneSpace
    alphaBoolBlend {blendInTime=0.3, blendOutTime=0.3}
    boneToModify {boneName=base_Ribbon_r_in_01}
    rotation=<rot -4 -18 5>
   }
   n33 = "Transform (Modify) Bone - Bone: base_Ribbon_l_mid_02"!(ComponentPose=@n32, bAlphaBoolEnabled:bool=True, Translation="0,0,0", Rotation="2.2,0,4", Scale="1,1,1") -> (Pose) {
    alphaInputType=Bool
    rotationMode=BMM_Additive
    rotationSpace=BCS_BoneSpace
    alphaBoolBlend {blendInTime=0.3, blendOutTime=0.3}
    boneToModify {boneName=base_Ribbon_l_mid_02}
    rotation {pitch=2.2, roll=4}
   }
   n34 = "Transform (Modify) Bone - Bone: base_Ribbon_r_mid_01"!(ComponentPose=@n33, bAlphaBoolEnabled:bool=True, Translation="0,0,0", Rotation="-2.2,0,-3.5", Scale="1,1,1") -> (Pose) {
    alphaInputType=Bool
    rotationMode=BMM_Additive
    rotationSpace=BCS_BoneSpace
    alphaBoolBlend {blendInTime=0.3, blendOutTime=0.3}
    boneToModify {boneName=base_Ribbon_r_mid_01}
    rotation {pitch=-2.2, roll=-3.5}
   }
   n35 = "Transform (Modify) Bone - Bone: pack"!(ComponentPose=@n34, bAlphaBoolEnabled:bool, Translation="0,-1.5,0", Rotation="0,0,0", Scale) -> (Pose) {alphaInputType=Bool, translationMode=BMM_Additive, translationSpace=BCS_BoneSpace, boneToModify {boneName=pack}, translation {y=-1.5}}
   n36 = Local To Component(LocalPose=Component To Local(ComponentPose=@n35))
   n37 = Get backpack_translation()
   n38 = Get backpack_rotate()
   n39 = Control Rig!(
    Source=Component To Local(ComponentPose="Transform (Modify) Bone - Bone: pack"!(ComponentPose=@n36, Alpha:float, Translation=@n37, Rotation=@n38, Scale) -> (Pose) {rotationMode=BMM_Additive, translationMode=BMM_Additive, boneToModify {boneName=pack}})
    Alpha
    Skydive_WaveSpeed:double=Get Skydive_Seed_CtrlRig() -> (Skydive_Seed_CtrlRig:double)
   ) -> (Pose) {controlRigAssetReference {blueprintRigClass=&/F_MED_MajorSpeech_Pack_CtrlRig.F_MED_MajorSpeech_Pack_CtrlRig_C}}
   Output Pose(Result=@n39) {layerGroup=None}
  ]}
 EventGraph {nodes [
   Event BlueprintUpdateAnimation()
   > n1 = Sequence!(execute) -> (then_0, then_1, then_2:Exec)
   > n2 = Branch(Condition=Get bIsSkyDiving())
   > Set Sim_space_settings(Sim_space_settings:Anim Phys Sim Space Settings Structure=Get Sim_space_settings_0() -> (Sim_space_settings_0:Anim Phys Sim Space Settings Structure)) -> (Output_Get:Anim Phys Sim Space Settings Structure)
   > Set Skydive_alpha(Skydive_alpha:double=0.5) -> (Output_Get:double)
   > Set Skydive_Seed_CtrlRig(Skydive_Seed_CtrlRig:double=4.7) -> (Output_Get:double)
   Set Sim_space_settings(execute=@n2.else, Sim_space_settings=Get Sim_space_settings_1() -> (Sim_space_settings_1:Anim Phys Sim Space Settings Structure))
   > Set Skydive_alpha(Skydive_alpha=1)
   > Set Skydive_Seed_CtrlRig(Skydive_Seed_CtrlRig=1)
   > Set JumpFall(JumpFall:bool=Get bIsFalling()) -> (Output_Get:bool)
   n3 = GetOwnerSkeletalMeshComponent()
   n4 = BreakRotator(InRot=BreakTransform(InTransform=GetDeltaTransformFromRefPose(self=@n3, BoneName=spine_05, BaseName=spine_01)).Rotation)
   Set backpack_rotate(
    execute=@n1.then_1
    backpack_rotate=MakeRotator(Roll=MapRangeClamped(Value=@n4.Yaw, InRangeA=0, InRangeB=-35, OutRangeA=3.4, OutRangeB=70), Yaw=0.5)
   )
   > Set backpack_translation(backpack_translation=MakeVector(Y=MapRangeClamped(Value=@n4.Yaw, InRangeA=0, InRangeB=-30, OutRangeB=-10)))
   n5 = Set Owning Pawn(execute=@n1.then_2, Owning Pawn:Pawn Object Reference=TryGetFortPawn()) -> (Output_Get:Pawn Object Reference)
   > Is Valid(InputObject=@n5)
   > n6 = Sequence!(execute) -> (then_0, then_1)
   > n7 = FindGender!(execute, Object:Object Reference=Get Owning Pawn() -> (Owning Pawn:Pawn Object Reference)) -> (True:Exec, False:Exec, OutGender:EFortCustomGender Enum=Invalid)
   > Set isMale(isMale:bool="Equal (Enum)"(A:EFortCustomGender Enum=@n7, B:EFortCustomGender Enum=Male)) -> (Output_Get:bool)
   n8 = FindBodyType!(execute=@n6.then_1, Object:Object Reference=Get Owning Pawn()) -> (True:Exec, False:Exec, OutBodyType:EFortCustomBodyType Enum=NONE)
   > Set isLarge(isLarge:bool="Equal (Enum)"(A:EFortCustomBodyType Enum=@n8, B:EFortCustomBodyType Enum=Large)) -> (Output_Get:bool)
   TryGetPawnOwner()
  ]}
]
variables [
 backpack_rotate {type=Rotator}
 backpack_translation {type=Vector}
 isLarge {type=boolean}
 isMale {type=boolean}
 JumpFall {type=boolean}
 Owning Pawn {type=Pawn, default=null}
 Sim_space_settings {
  type=AnimPhysSimSpaceSettings
  default {maxAngularAcceleration=50, maxAngularVelocity=5, simSpaceAngularAlpha=1, externalAngularVelocity=<0 0 0>}
 }
 Sim_space_settings_0 {
  type=AnimPhysSimSpaceSettings
  default {maxAngularAcceleration=50, maxAngularVelocity=5, simSpaceAngularAlpha=0, externalAngularVelocity=<0 0 0>}
 }
 Sim_space_settings_1 {
  type=AnimPhysSimSpaceSettings
  default {maxAngularAcceleration=50, maxAngularVelocity=5, simSpaceAngularAlpha=0.8, externalAngularVelocity=<0 0 0>}
 }
 Skydive_alpha {type=number}
 Skydive_Seed_CtrlRig {type=number}
]
